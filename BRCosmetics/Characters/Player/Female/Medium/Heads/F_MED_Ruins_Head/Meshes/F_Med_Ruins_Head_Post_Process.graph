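# Post-process AnimBlueprint for the female Ruins head on the Fortnite base head skeleton, adjusting extra face bones.
# The input pose runs through the F_MED_Ruins_Head Control Rig, then a constraint ties face_04_l rotation to L_eye before output.
# A face_01_l eyelid constraint and a face_01_r offset sit unconnected.
# Initialization caches the mesh component, while the eyelid calculation from bone deltas is not wired to the empty update event.
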AnimBlueprint /BRCosmetics/Characters/Player/Female/Medium/Heads/F_MED_Ruins_Head/Meshes/F_Med_Ruins_Head_Post_Process
target_skeleton=/Game/Characters/Player/Common/Fortnite_Base_Head/Fortnite_Base_Head_Export_Skeleton
graphs [
 AnimGraph {nodes [
   "Constraint - face_01_l (TR)"!(ComponentPose, Alpha=1, ConstraintWeights_0=1, ConstraintWeights_1:float=1) -> (Pose) {
    alphaBoolBlend {blendOption=Linear}
    boneToModify {boneName=face_01_l}
    constraintSetup [{transformType=Translation, targetBone {boneName=L_eye_lid_lower_mid}}, {transformType=Rotation, targetBone {boneName=L_eye_lid_lower_mid}}]
    constraintWeights [1,1]
   }
   "Transform (Modify) Bone - Bone: face_01_r"(Alpha=1, Translation="0.000,0,-1", Rotation="0.000,-0.000,-40", Scale="1,1,1") {
    rotationMode=BMM_Additive
    rotationSpace=BCS_BoneSpace
    translationMode=BMM_Additive
    translationSpace=BCS_BoneSpace
    alphaBoolBlend {blendOption=Linear}
    boneToModify {boneName=face_01_r}
    rotation {roll=-40}
    translation {z=-1}
   }
   n1 = Control Rig(Source=Input Pose()) {
    alphaBoolBlend {blendOption=Linear}
    controlRigAssetReference {blueprintRigClass=&/F_MED_Ruins_Head_CtrlRig.F_MED_Ruins_Head_CtrlRig_C}
   }
   n2 = "Constraint - face_04_l (R)"(ComponentPose=Local To Component(LocalPose=@n1), Alpha=1, ConstraintWeights_0=1) {
    alphaBoolBlend {blendOption=Linear}
    boneToModify {boneName=face_04_l}
    constraintSetup [{transformType=Rotation, targetBone {boneName=L_eye}}]
    constraintWeights [1]
   }
   Output Pose(Result=Component To Local(ComponentPose=@n2)) {layerGroup=None}
  ]}
 EventGraph {nodes [
   Event BlueprintInitializeAnimation()
   > Set Skel_Mesh(Skel_Mesh=GetOwnerSkeletalMeshComponent())
   TryGetPawnOwner()
   n1 = BreakTransform(InTransform=GetDeltaTransformFromRefPose(self=GetOwnerSkeletalMeshComponent(), BoneName=L_eye_lid_upper_mid, BaseName=head))
   BreakRotator(InRot=@n1.Rotation)
   TryGetPawnOwner()
   GetAlphaFromReferenceTransformDelta(
    SkeletalMeshComponent=Get Skel_Mesh()
    BoneName=L_eye_lid_upper_mid
    BaseName=L_eye
    DeltaComponentType=TranslationZ
    InputRange=(X=0,Y=1)
    TwistAxis="1,0,0"
   )
   Event BlueprintUpdateAnimation()
   Set eyelid(eyelid:double=MapRangeClamped(Value=BreakVector(InVec=@n1).Z, InRangeA=-15, InRangeB=-100)) -> (Output_Get:double)
  ]}
]
variables [
 eyelid {type=number}
 Skel_Mesh {type=SkeletalMeshComponent}
]
